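# Animation Blueprint for the MorningSoak Cook large male head, built around flipbook face animation.
# Its event graph compares mouth, eye and brow flipbook curves from the part mesh's anim instance against cached indices.
# A node chain raises the material's Temporal Responsiveness when any index changes, likely to limit temporal smearing, then caches the new values.
# The update event shows only a parent call, with no visible link to that chain.
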AnimBlueprint /BRCosmetics/Characters/Player/Male/Large/Heads/M_LRG_MorningSoak_Cook_Head/Meshes/M_LRG_MorningSoak_Cook_Head_AnimBP
target_skeleton=&/Fortnite_Base_Head_MorningSoak_Cook_Skeleton
graphs [EventGraph {nodes [
   n1 = GetPartSkeletalMeshComponent()
   n2 = GetAnimInstance(self=@n1)
   n3 = GetCurveValue(self=@n2, CurveName=flipbook_face_mouth_index)
   n4 = "Not Equal (Float)"(A=@n3, B=Get Mouth Index() -> (Mouth Index:float))
   n5 = GetCurveValue(self=@n2, CurveName=flipbook_face_L_brow_index)
   n6 = "Not Equal (Float)"(A=@n5, B=Get L_Brow_Index() -> (L_Brow_Index:float))
   n7 = GetCurveValue(self=@n2, CurveName=flipbook_face_R_eye_index)
   n8 = "Not Equal (Float)"(A=@n7, B=Get Right Eye Index() -> (Right Eye Index:float))
   n9 = GetCurveValue(self=@n2, CurveName=flipbook_face_L_eye_index)
   n10 = "Not Equal (Float)"(A=@n9, B=Get Left Eye Index() -> (Left Eye Index:float))
   n11 = GetCurveValue(self=@n2, CurveName=flipbook_face_R_brow_index)
   n12 = "Not Equal (Float)"(A=@n11, B=Get R_Brow_Index() -> (R_Brow_Index:float))
   SetScalarParameterValueOnMaterials!(
    execute
    self:Mesh Component Object Reference=@n1
    ParameterName:Name=Temporal Responsiveness
    ParameterValue:float=SelectFloat(A=1, B=0.5, bPickA=OR Boolean!(A=@n8, B=@n10, C:bool=@n4, D:bool=@n6, E:bool=@n12) -> (ReturnValue))
   ) -> (then)
   > Set Left Eye Index(Left Eye Index:float=@n9) -> (Output_Get:float)
   > Set Right Eye Index(Right Eye Index:float=@n7) -> (Output_Get:float)
   > Set Mouth Index(Mouth Index:float=@n3) -> (Output_Get:float)
   > Set L_Brow_Index(L_Brow_Index:float=@n5) -> (Output_Get:float)
   > Set R_Brow_Index(R_Brow_Index:float=@n11) -> (Output_Get:float)
   n13 = Event BlueprintUpdateAnimation()
   > Parent: BlueprintUpdateAnimation!(execute, DeltaTimeX:float=@n13.DeltaTimeX) -> (then) {is=title K2Node_CallParentFunction}
   TryGetPawnOwner()
   Parent: BlueprintUpdateAnimation(DeltaTimeX=0)
  ]}]
variables [
 L_Brow_Index {type=number}
 Left Eye Index {type=number}
 Mouth Index {type=number}
 R_Brow_Index {type=number}
 Right Eye Index {type=number}
]
