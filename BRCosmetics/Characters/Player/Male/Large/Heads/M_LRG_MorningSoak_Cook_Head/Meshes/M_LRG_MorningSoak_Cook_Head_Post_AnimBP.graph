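# Post-process Animation Blueprint for the MorningSoak Cook male head that corrects head and neck posing and deforms it with body motion.
# Each update scales and negates head rotation relative to neck_02's ref-pose delta, adding it to both bones to dampen movement.
# A Control Rig then deforms that pose using upper-arm-raise curves read from the body mesh's anim instance.
# In the front end, the frontend_idle_bypass curve skips the Control Rig during custom idles.
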AnimBlueprint /BRCosmetics/Characters/Player/Male/Large/Heads/M_LRG_MorningSoak_Cook_Head/Meshes/M_LRG_MorningSoak_Cook_Head_Post_AnimBP
target_skeleton=&/Fortnite_Base_Head_MorningSoak_Cook_Skeleton
graphs [
 AnimGraph {nodes [
   Output Pose(Result=Blend Poses by bool(
     BlendPose_0=Use cached pose 'StartPose'()
     BlendPose_1=Use cached pose 'DeformPose'()
     bActiveValue=Get Custom Idle Is Playing() -> (Custom Idle Is Playing:bool)
    ))
   n1 = Get HeadRot() -> (HeadRot:Rotator)
   n2 = "Transform (Modify) Bone - Bone: head"(ComponentPose=Local To Component(LocalPose=Input Pose()), Alpha=1, Translation="0,0,0", Rotation=@n1, Scale="1,1,1") {rotationMode=BMM_Additive, rotationSpace=BCS_BoneSpace, boneToModify {boneName=head}}
   Save cached pose 'StartPose'(Pose=Component To Local(ComponentPose="Transform (Modify) Bone - Bone: neck_02"(ComponentPose=@n2, Alpha=1, Translation="0,0,0", Rotation=@n1, Scale="1,1,1") {rotationMode=BMM_Additive, rotationSpace=BCS_BoneSpace, boneToModify {boneName=neck_02}})) {cacheName=StartPose}
   n3 = Control Rig!(
    Source=Use cached pose 'StartPose'()
    Alpha
    l_arm_up_alpha:double=Get upperarm_l_up_alpha() -> (upperarm_l_up_alpha:double)
    r_arm_up_alpha:double=Get upperarm_r_up_alpha() -> (upperarm_r_up_alpha:double)
   ) -> (Pose) {controlRigAssetReference {blueprintRigClass=&/M_LRG_MorningSoak_Cook_Head_CtrlRig.M_LRG_MorningSoak_Cook_Head_CtrlRig_C}}
   Save cached pose 'DeformPose'(Pose=@n3) {cacheName=DeformPose}
  ]}
 EventGraph {nodes [
   Event BlueprintUpdateAnimation()
   > n1 = Sequence()
   > Branch(Condition=Get bIsInFrontEnd())
   > Set Custom Idle Is Playing Float(Custom Idle Is Playing Float:double=GetCurveValue(self=GetOwnerFrontEndAnimInstance(), CurveName=frontend_idle_bypass)) -> (Output_Get:double)
   > n2 = Branch(Condition="float >= float"(A=Get Custom Idle Is Playing Float() -> (Custom Idle Is Playing Float:double), B=1))
   > Set Custom Idle Is Playing(Custom Idle Is Playing:bool=true) -> (Output_Get:bool)
   Set Custom Idle Is Playing(execute=@n2.else, Custom Idle Is Playing=false)
   n3 = Get BodyMesh() -> (BodyMesh:Skeletal Mesh Component Budgeted Object Reference)
   Is Valid(exec=@n1.then_1, InputObject=@n3)
   n4 = GetAnimInstance(self=@n3)
   > Set upperarm_l_up_alpha(upperarm_l_up_alpha:double=GetCurveValue(self=@n4, CurveName=upperarm_l_up_SPR)) -> (Output_Get:double)
   > Set upperarm_r_up_alpha(upperarm_r_up_alpha:double=GetCurveValue(self=@n4, CurveName=upperarm_r_up_SPR)) -> (Output_Get:double)
   n5 = GetOwnerSkeletalMeshComponent()
   n6 = BreakRotator(InRot=BreakTransform(InTransform=GetDeltaTransformFromRefPose(self=@n5, BoneName=head, BaseName=neck_02)).Rotation)
   n7 = Get Z Percentage() -> (Z Percentage:double)
   n8 = Get X Percentage() -> (X Percentage:double)
   n9 = Get Y Percentage() -> (Y Percentage:double)
   > Set HeadRot(HeadRot:Rotator=MakeRotator(
     Roll=float * float!(A=@n6, B=-1, C:double=@n8) -> (ReturnValue) {is=title K2Node_PromotableOperator}
     Pitch=float * float(A=@n6.Pitch, B=-1, C=@n9)
     Yaw=float * float(A=@n6.Yaw, B=-1, C=@n7)
    )) -> (Output_Get:Rotator)
   n10 = Event OnCharacterCustomizationLoaded()
   > Set BodyMesh(BodyMesh:Skeletal Mesh Component Budgeted Object Reference=Get Array Item with EFort Custom Part Type(Array=@n10.CustomMeshParts)) -> (Output_Get:Skeletal Mesh Component Budgeted Object Reference)
   TryGetPawnOwner()
  ]}
]
variables [
 BodyMesh {type=SkeletalMeshComponentBudgeted}
 Custom Idle Is Playing {type=boolean}
 Custom Idle Is Playing Float {type=number}
 HeadRot {type=Rotator}
 upperarm_l_up_alpha {type=number}
 upperarm_r_up_alpha {type=number}
 X Percentage {type=number, default=0.2}
 Y Percentage {type=number, default=0.2}
 Z Percentage {type=number, default=0.3}
]
